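; Gameplay ability for the Lawless shotgun exotic's slap energy drink effect, a regen ability by name.
; On activation it runs the parent ability's logic, which is not shown here.
; It then strips all instances of its gameplay cue tag from the player pawn at index 0 through an AI common-functions library helper.
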(event Ability|EventActivateAbility
  (|CallParentFunction)
  (Class|BPFLAICommonFunctions|RemoveAllInstancesOfGameplayTag (Game|GetPlayerPawn 0) (Variables|Default|GetGameplay%20Cue%20Tag)))
